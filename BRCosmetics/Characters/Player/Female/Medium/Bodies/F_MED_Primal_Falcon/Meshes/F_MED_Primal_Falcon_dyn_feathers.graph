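# Physics asset for the dynamic feathers on the F_MED_Primal_Falcon female body, covering only left-side skirt bones.
# Back, side and front chains each start at a tiny master capsule, constrained to the left thigh twist bone with free swing and twist.
# Feather segments below have tight twist and swing1 limits and wider swing2, loosest at the back; the front has two segments.
# Thigh and calf capsules provide leg bodies alongside them.
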PhysicsAsset /BRCosmetics/Characters/Player/Female/Medium/Bodies/F_MED_Primal_Falcon/Meshes/F_MED_Primal_Falcon_dyn_feathers
bodies {
 calf_l [{
   shapeType=Capsule
   length=37.5
   radius=8
   shapeName=thigh_twist_01_l_capsule
   center=<-24.3267 -2.4035 0.5327>
   rotation=<rot 90 12.9074 7.9075>
  }]
 dyn_skirt_bk_l_01 [{shapeType=Capsule, length=13.5, radius=6.5, shapeName=dyn_skirt_bk_l_01_capsule, center=<10 0 2>, rotation=<rot 90 0 0>}]
 dyn_skirt_bk_l_MSTR [{shapeType=Capsule, length=0.505, radius=0.505, shapeName=dyn_skirt_bk_l_MSTR_capsule, rotation=<rot 0 -90 0>}]
 dyn_skirt_ft_l_01 [{shapeType=Capsule, length=6.5, radius=7, shapeName=dyn_skirt_bk_l_01_capsule, center=<8 0 4>, rotation=<rot 90 0 0>}]
 dyn_skirt_ft_l_02 [{shapeType=Capsule, length=6, radius=7.5, shapeName=dyn_skirt_bk_l_01_capsule, center=<8 0 4>, rotation=<rot 90 0 0>}]
 dyn_skirt_ft_l_MSTR [{shapeType=Capsule, length=0.505, radius=0.505, shapeName=dyn_skirt_ft_l_MSTR_capsule, rotation=<rot 0 -90 0>}]
 dyn_skirt_l_01 [{shapeType=Capsule, length=13.25, radius=6.5, shapeName=dyn_skirt_bk_l_01_capsule, center=<10 0 5>, rotation=<rot 90 0 0>}]
 dyn_skirt_l_MSTR [{shapeType=Capsule, length=0.505, radius=0.505, shapeName=dyn_skirt_l_MSTR_capsule, rotation=<rot 0 -90 0>}]
 thigh_twist_01_l [{
   shapeType=Capsule
   length=26
   radius=7.75
   shapeName=thigh_twist_01_l_capsule
   center=<2.6522 -2.2382 -0.9644>
   rotation=<rot 90 0.2373 0.2374>
  }]
}
constraints [
 {
  bone1Name=dyn_skirt_bk_l_MSTR
  bone2Name=thigh_twist_01_l
  swing1LimitDegrees=45
  swing1Motion=Free
  swing2LimitDegrees=45
  swing2Motion=Free
  twistLimitDegrees=45
  twistMotion=Free
 }
 {bone1Name=dyn_skirt_bk_l_01, bone2Name=dyn_skirt_bk_l_MSTR, swing1LimitDegrees=5, swing2LimitDegrees=35, twistLimitDegrees=2}
 {
  bone1Name=dyn_skirt_l_MSTR
  bone2Name=thigh_twist_01_l
  swing1LimitDegrees=45
  swing1Motion=Free
  swing2LimitDegrees=45
  swing2Motion=Free
  twistLimitDegrees=45
  twistMotion=Free
 }
 {bone1Name=dyn_skirt_l_01, bone2Name=dyn_skirt_l_MSTR, swing1LimitDegrees=5, swing2LimitDegrees=25, twistLimitDegrees=2}
 {
  bone1Name=dyn_skirt_ft_l_MSTR
  bone2Name=thigh_twist_01_l
  swing1LimitDegrees=45
  swing1Motion=Free
  swing2LimitDegrees=45
  swing2Motion=Free
  twistLimitDegrees=45
  twistMotion=Free
 }
 {bone1Name=dyn_skirt_ft_l_01, bone2Name=dyn_skirt_ft_l_MSTR, swing1LimitDegrees=5, swing2LimitDegrees=15, twistLimitDegrees=2}
 {bone1Name=dyn_skirt_ft_l_02, bone2Name=dyn_skirt_ft_l_01, swing1LimitDegrees=5, swing2LimitDegrees=15, twistLimitDegrees=2}
]
