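# Physics asset for the neck feather secondary motion on the F_MED_Primal_Falcon player body.
# It defines four feather chains (left, right, back, back-low), each pairing a larger offset capsule on the feather bone with a small capsule on its MSTR bone.
# One constraint ties each feather to its MSTR bone, keeping twist and one swing axis tight while allowing a much wider second swing, so each feather mostly flaps along one axis.
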PhysicsAsset /BRCosmetics/Characters/Player/Female/Medium/Bodies/F_MED_Primal_Falcon/Meshes/F_MED_Primal_Falcon_dyn_feathers_neck
bodies {
 dyn_neck_feather_bk [{shapeType=Capsule, length=10, radius=4, shapeName=dyn_neck_feather_bk_capsule, center=<6 0 0>, rotation=<rot 0 0 90>}]
 dyn_neck_feather_bk_MSTR [{shapeType=Capsule, length=1, radius=1}]
 dyn_neck_feather_bk_low [{shapeType=Capsule, length=10, radius=4, shapeName=dyn_neck_feather_bk_capsule, center=<6 0 -1>, rotation=<rot 0 0 90>}]
 dyn_neck_feather_bk_low_MSTR [{shapeType=Capsule, length=1, radius=1}]
 dyn_neck_feather_l [{shapeType=Capsule, length=10, radius=4, shapeName=dyn_neck_feather_l_capsule, center=<4.5 0 0>, rotation=<rot 0 0 90>}]
 dyn_neck_feather_l_MSTR [{shapeType=Capsule, length=1, radius=1}]
 dyn_neck_feather_r [{shapeType=Capsule, length=10, radius=4, shapeName=dyn_neck_feather_r_capsule, center=<5 0 0>, rotation=<rot 0 0 90>}]
 dyn_neck_feather_r_MSTR [{shapeType=Capsule, length=1, radius=1}]
}
constraints [
 {bone1Name=dyn_neck_feather_r, bone2Name=dyn_neck_feather_r_MSTR, swing1LimitDegrees=5, swing2LimitDegrees=45, twistLimitDegrees=2}
 {bone1Name=dyn_neck_feather_l, bone2Name=dyn_neck_feather_l_MSTR, swing1LimitDegrees=5, swing2LimitDegrees=45, twistLimitDegrees=2}
 {bone1Name=dyn_neck_feather_bk, bone2Name=dyn_neck_feather_bk_MSTR, swing1LimitDegrees=5, swing2LimitDegrees=45, twistLimitDegrees=2}
 {bone1Name=dyn_neck_feather_bk_low, bone2Name=dyn_neck_feather_bk_low_MSTR, swing1LimitDegrees=5, swing2LimitDegrees=45, twistLimitDegrees=2}
]
