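# Post-pose AnimBP for the Primal Falcon body cosmetic: copies the attached parent's pose, then deforms skirt, belt, mask, pads and feathers.
# Two Control Rigs run first, followed by constraints, look-ats, additive bone offsets, AnimDynamics on the mask and two RigidBody feather sims.
# Event update maps thigh, right arm and hand rotations from ref pose into 0–1 offset alphas.
# Skydive and dive states switch the shoulder-pad alphas.
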AnimBlueprint /BRCosmetics/Characters/Player/Female/Medium/Bodies/F_MED_Primal_Falcon/Meshes/F_MED_Primal_Falcon_AnimBP
target_skeleton=&/F_MED_Primal_Falcon_Skeleton
graphs [
 AnimGraph {nodes [
   n1 = Control Rig(Source=Copy Pose From Mesh() {bUseAttachedParent=true}) {
    alphaBoolBlend {blendOption=Linear}
    controlRigAssetReference {blueprintRigClass=&/F_MED_Primal_Falcon_Collector_Deform_CtrlRig_twists.F_MED_Primal_Falcon_Collector_Deform_CtrlRig_twists_C}
   }
   n2 = Control Rig!(
    Source=@n1
    Alpha
    KneeRot_fr_l_114deg:Vector
    KneeTrans_fr_l_114deg:Vector="0.258,-3.591,0.009"
    KneeScale_fr_l_114deg:Vector="1,1,1"
    KneeScale_fr_l_130deg:Vector="1,1,1"
    KneeRot_fr_l_130deg:Vector="0,0,-51.2"
    KneeTrans_fr_l_130deg:Vector="0.258,-3.591,0.009"
   ) -> (Pose) {
    alphaBoolBlend {blendOption=Linear}
    controlRigAssetReference {blueprintRigClass=&/F_MED_Primal_Falcon_Deform_CtrlRig_sculpters.F_MED_Primal_Falcon_Deform_CtrlRig_sculpters_C}
   }
   n3 = "Constraint - dyn_skirt_bk_l_MSTR (PP)"!(ComponentPose=Local To Component(LocalPose=@n2), Alpha=1, ConstraintWeights_0=1, ConstraintWeights_1:float=0) -> (Pose) {
    alphaBoolBlend {blendOption=Linear}
    boneToModify {boneName=dyn_skirt_bk_l_MSTR}
    constraintSetup [{targetBone {boneName=pelvis}}, {targetBone {boneName=thigh_twist_01_l}}]
    constraintWeights [1,1]
   }
   n4 = "Constraint - def_packet_l_MSTR (PP)"!(ComponentPose=@n3, Alpha=1, ConstraintWeights_0=0.9768, ConstraintWeights_1:float=0.0232) -> (Pose) {
    alphaBoolBlend {blendOption=Linear}
    boneToModify {boneName=def_packet_l_MSTR}
    constraintSetup [{targetBone {boneName=pelvis}}, {targetBone {boneName=spine_01}}]
    constraintWeights [1,1]
   }
   n5 = Get thigh_rot_up_l() -> (thigh_rot_up_l:double)
   n6 = "Transform (Modify) Bone - Bone: def_packet_l"(ComponentPose=@n4, Alpha=@n5, Translation="0,-2,0", Rotation="2,-15,-17", Scale="1,1,1") {
    rotationMode=BMM_Additive
    rotationSpace=BCS_BoneSpace
    translationMode=BMM_Additive
    translationSpace=BCS_BoneSpace
    alphaBoolBlend {blendOption=Linear}
    boneToModify {boneName=def_packet_l}
    rotation=<rot 2 -15 -17>
    translation {y=-2}
   }
   n7 = "Transform (Modify) Bone - Bone: dyn_skirt_ft_l_MSTR"(ComponentPose=@n6, Alpha=@n5, Translation="0,0,0", Rotation="0,-20,0", Scale="1,1,1") {
    rotationMode=BMM_Additive
    rotationSpace=BCS_BoneSpace
    translationMode=BMM_Additive
    translationSpace=BCS_BoneSpace
    alphaBoolBlend {blendOption=Linear}
    boneToModify {boneName=dyn_skirt_ft_l_MSTR}
    rotation {yaw=-20}
   }
   n8 = "Transform (Modify) Bone - Bone: dyn_skirt_l_MSTR"(ComponentPose=@n7, Alpha=@n5, Translation="-8,0,5", Rotation="0,-80,0", Scale="1,1,1") {
    rotationMode=BMM_Additive
    rotationSpace=BCS_BoneSpace
    translationMode=BMM_Additive
    translationSpace=BCS_BoneSpace
    alphaBoolBlend {blendOption=Linear}
    boneToModify {boneName=dyn_skirt_l_MSTR}
    rotation {yaw=-80}
    translation {x=-8, z=5}
   }
   n9 = "Transform (Modify) Bone - Bone: def_belt_l"(ComponentPose=@n8, Alpha=@n5, Translation="12,0,0", Rotation="10,-22,9.5", Scale="1,1,1") {
    rotationMode=BMM_Additive
    rotationSpace=BCS_BoneSpace
    translationMode=BMM_Additive
    translationSpace=BCS_BoneSpace
    alphaBoolBlend {blendOption=Linear}
    boneToModify {boneName=def_belt_l}
    rotation=<rot 10 -22 9.5>
    translation {x=12}
   }
   n10 = "Transform (Modify) Bone - Bone: dyn_mask_MSTR"(
    ComponentPose=@n9
    Alpha=Get thigh_rot_out_r() -> (thigh_rot_out_r:double)
    Translation="1.124,3.633,1.597"
    Rotation="5.202,-8.633,16.310"
    Scale="1,1,1"
   ) {
    rotationMode=BMM_Additive
    rotationSpace=BCS_BoneSpace
    translationMode=BMM_Additive
    translationSpace=BCS_BoneSpace
    alphaBoolBlend {blendOption=Linear}
    boneToModify {boneName=dyn_mask_MSTR}
    rotation=<rot 5.202 -8.6331 16.3101>
    translation=<1.124 3.633 1.597>
   }
   n11 = Get thigh_rot_up_r() -> (thigh_rot_up_r:double)
   n12 = "Transform (Modify) Bone - Bone: dyn_mask_MSTR"(ComponentPose=@n10, Alpha=@n11, Translation="2,0,3.5", Rotation="-20,0.000,0", Scale="1,1,1") {
    rotationMode=BMM_Additive
    rotationSpace=BCS_BoneSpace
    translationMode=BMM_Additive
    translationSpace=BCS_BoneSpace
    alphaBoolBlend {blendOption=Linear}
    boneToModify {boneName=dyn_mask_MSTR}
    rotation {pitch=-20}
    translation {x=2, z=3.5}
   }
   n13 = "Transform (Modify) Bone - Bone: def_belt_r"(ComponentPose=@n12, Alpha=@n11, Translation="6,0,0", Rotation="0,0,0", Scale="1,1,1") {
    rotationMode=BMM_Additive
    rotationSpace=BCS_BoneSpace
    translationMode=BMM_Additive
    translationSpace=BCS_BoneSpace
    alphaBoolBlend {blendOption=Linear}
    boneToModify {boneName=def_belt_r}
    translation {x=6}
   }
   n14 = "Transform (Modify) Bone - Bone: def_mask_blet_01"(ComponentPose=@n13, Alpha=@n11, Translation="-1.904,0,1.370", Rotation="0,0,0", Scale="1,1,1") {
    rotationMode=BMM_Additive
    rotationSpace=BCS_BoneSpace
    translationMode=BMM_Additive
    translationSpace=BCS_BoneSpace
    alphaBoolBlend {blendOption=Linear}
    boneToModify {boneName=def_mask_blet_01}
    translation {x=-1.904, z=1.37}
   }
   n15 = "Constraint - def_mask_blet_02_MSTR (PP)"!(ComponentPose=@n14, Alpha=1, ConstraintWeights_0=0.8889, ConstraintWeights_1:float=0.1111) -> (Pose) {
    alphaBoolBlend {blendOption=Linear}
    boneToModify {boneName=def_mask_blet_02_MSTR}
    constraintSetup [{targetBone {boneName=pelvis}}, {targetBone {boneName=def_belt_r}}]
    constraintWeights [1,1]
   }
   n16 = "Constraint - def_mask_blet_01_MSTR (PPP)"!(ComponentPose=@n15, Alpha=1, ConstraintWeights_0=0.1027, ConstraintWeights_1:float=0.3429, ConstraintWeights_2:float=0.5543) -> (Pose) {
    alphaBoolBlend {blendOption=Linear}
    boneToModify {boneName=def_mask_blet_01_MSTR}
    constraintSetup [{targetBone {boneName=pelvis}}, {targetBone {boneName=def_belt_l}}, {targetBone {boneName=def_belt_r}}]
    constraintWeights [1,1,1]
   }
   n17 = Look At - Bone: dyn_mask_AIM(ComponentPose=@n16, Alpha=1) {
    alphaBoolBlend {blendOption=Linear}
    boneToModify {boneName=dyn_mask_AIM}
    lookAtLocation {x=0}
    lookAtTarget {boneReference {boneName=dyn_mask_TGT}}
    lookAt_Axis {axis=<1 0>}
    lookUp_Axis {axis {x=0}}
   }
   n18 = Look At - Bone: def_mask_blet_01(ComponentPose=@n17, Alpha=1) {
    alphaBoolBlend {blendOption=Linear}
    boneToModify {boneName=def_mask_blet_01}
    lookAtLocation {x=0}
    lookAtTarget {boneReference {boneName=def_mask_blet_01_TGT}}
    lookAt_Axis {axis=<1 0>}
    lookUp_Axis {axis {x=0}}
   }
   n19 = Look At - Bone: def_mask_blet_02(ComponentPose=@n18, Alpha=1) {
    alphaBoolBlend {blendOption=Linear}
    boneToModify {boneName=def_mask_blet_02}
    lookAtLocation {x=0}
    lookAtTarget {boneReference {boneName=def_mask_blet_02_TGT}}
    lookAt_Axis {axis=<1 0>}
    lookUp_Axis {axis {x=0}}
   }
   n20 = Anim Dynamics - Bone: dyn_mask(ComponentPose=@n19, Alpha=1, ExternalForce="0,0,0") {
    angularDampingOverride=0.9999
    bGravityOverrideInSimSpace=true
    bOverrideAngularDamping=true
    bUseGravityOverride=true
    linearDampingOverride=0.9
    simulationSpace=RootRelative
    alphaBoolBlend {blendOption=Linear}
    alphaScaleBiasClamp {scale=0.6}
    boundBone {boneName=dyn_mask}
    componentAppliedLinearAccClamp {z=5000}
    componentLinearAccScale {z=0.4}
    componentLinearVelScale {z=0.5}
    gravityOverride {z=-980}
    physicsBodyDefinitions [{
      sphereCollisionRadius=0
      boundBone {boneName=dyn_mask}
      boxExtents=<10 20 15>
      constraintSetup {angularLimitsMax=<0 5 3>, angularLimitsMin=<0 -5 -3>, angularTarget=<0 0 0>}
      localJointOffset=<5 0 0>
     }]
   }
   n21 = "Transform (Modify) Bone - Bone: def_mask_blet_02"(ComponentPose=@n20, Alpha=Get thigh_rot_innr_r() -> (thigh_rot_innr_r:double), Translation="0,0,0", Rotation="20,0.000,0", Scale="1,1,1") {
    rotationMode=BMM_Additive
    rotationSpace=BCS_BoneSpace
    translationMode=BMM_Additive
    translationSpace=BCS_BoneSpace
    alphaBoolBlend {blendOption=Linear}
    boneToModify {boneName=def_mask_blet_02}
    rotation {pitch=20}
   }
   n22 = "Constraint - def_shoulder_pad_MSTR (PP)"!(ComponentPose=@n21, Alpha=1, ConstraintWeights_0=0.15, ConstraintWeights_1:float=0.85) -> (Pose) {
    alphaBoolBlend {blendOption=Linear}
    boneToModify {boneName=def_shoulder_pad_MSTR}
    constraintSetup [{targetBone {boneName=spine_05}}, {targetBone {boneName=clavicle_r}}]
    constraintWeights [1,1]
   }
   n23 = Get should_pad_def_alpha() -> (should_pad_def_alpha:double)
   n24 = "Transform (Modify) Bone - Bone: def_shoulder_pad"(ComponentPose=@n22, Alpha=@n23, Translation="0,5,-1.5", Rotation="0,10,45", Scale="1,1,1") {
    rotationMode=BMM_Additive
    rotationSpace=BCS_BoneSpace
    translationMode=BMM_Additive
    translationSpace=BCS_BoneSpace
    alphaBoolBlend {blendOption=Linear}
    boneToModify {boneName=def_shoulder_pad}
    rotation {roll=45, yaw=10}
    translation {y=5, z=-1.5}
   }
   n25 = "Transform (Modify) Bone - Bone: def_shoulder_pad_ft"(ComponentPose=@n24, Alpha=@n23, Translation="0,0,0.5", Rotation="15,0,0", Scale="1,1,1") {
    rotationMode=BMM_Additive
    rotationSpace=BCS_BoneSpace
    translationMode=BMM_Additive
    translationSpace=BCS_BoneSpace
    alphaBoolBlend {blendOption=Linear}
    boneToModify {boneName=def_shoulder_pad_ft}
    rotation {pitch=15}
    translation {z=0.5}
   }
   n26 = "Transform (Modify) Bone - Bone: def_shoulder_pad"(
    ComponentPose=@n25
    Alpha=Get should_pad_def_alpha_dive() -> (should_pad_def_alpha_dive:double)
    Translation="0,0,0"
    Rotation="22,0,0"
    Scale="1,1,1"
   ) {
    rotationMode=BMM_Additive
    rotationSpace=BCS_BoneSpace
    translationMode=BMM_Additive
    translationSpace=BCS_BoneSpace
    alphaBoolBlend {blendOption=Linear}
    boneToModify {boneName=def_shoulder_pad}
    rotation {pitch=22}
   }
   n27 = "Transform (Modify) Bone - Bone: def_arm_pad_r"(
    ComponentPose=@n26
    Alpha=Get hand_rot_up_r() -> (hand_rot_up_r:double)
    Translation="-2.835,-0.000,0.690"
    Rotation="5.000,0.000,0.000"
    Scale="1,1,1"
   ) {
    rotationMode=BMM_Additive
    rotationSpace=BCS_BoneSpace
    translationMode=BMM_Additive
    translationSpace=BCS_BoneSpace
    alphaBoolBlend {blendOption=Linear}
    boneToModify {boneName=def_arm_pad_r}
    rotation {pitch=5}
    translation {x=-2.835, z=0.69}
   }
   n28 = "Constraint - def_knees_armor_l (PP)"!(ComponentPose=@n27, Alpha=1, ConstraintWeights_0=0.5, ConstraintWeights_1:float=0.5) -> (Pose) {
    alphaBoolBlend {blendOption=Linear}
    boneToModify {boneName=def_knees_armor_l}
    constraintSetup [{targetBone {boneName=thigh_l}}, {targetBone {boneName=calf_l}}]
    constraintWeights [1,1]
   }
   n29 = "Constraint - def_knees_armor_r (PP)"!(ComponentPose=@n28, Alpha=1, ConstraintWeights_0=0.5, ConstraintWeights_1:float=0.5) -> (Pose) {
    alphaBoolBlend {blendOption=Linear}
    boneToModify {boneName=def_knees_armor_r}
    constraintSetup [{targetBone {boneName=thigh_r}}, {targetBone {boneName=calf_r}}]
    constraintWeights [1,1]
   }
   n30 = RigidBody!(ComponentPose=@n29, Alpha, OverrideWorldGravity:Vector="0,0,-980", ExternalForce) -> (Pose) {
    bClampLinearTranslationLimitToRefPose=true
    bOverrideWorldGravity=true
    overridePhysicsAsset=&/F_MED_Primal_Falcon_dyn_feathers
    simulationSpace=BaseBoneSpace
    alphaBoolBlend {blendOption=Linear}
    baseBoneRef {boneName=root}
    componentAppliedLinearAccClamp=<3000 3000 3000>
    componentLinearAccScale {z=0.5}
    componentLinearVelScale {z=0.5}
    overrideWorldGravity {z=-980}
    simSpaceSettings {
     maxAngularAcceleration=800
     maxAngularVelocity=5
     maxLinearAcceleration=1000
     maxLinearVelocity=200
     velocityScaleZ=0
     worldAlpha=1
     externalLinearDragV=<1 1 1>
    }
   }
   n31 = RigidBody(ComponentPose=@n30, OverrideWorldGravity="0,0,-100") {
    bClampLinearTranslationLimitToRefPose=true
    bOverrideWorldGravity=true
    overridePhysicsAsset=&/F_MED_Primal_Falcon_dyn_feathers_neck
    simulationSpace=BaseBoneSpace
    alphaBoolBlend {blendOption=Linear}
    baseBoneRef {boneName=neck_01}
    componentAppliedLinearAccClamp=<1000 1000 1000>
    componentLinearAccScale=<0.1 0.1 0.1>
    componentLinearVelScale=<0.1 0.1 0.1>
    overrideWorldGravity {z=-100}
   }
   Output Pose(Result=Component To Local(ComponentPose=@n31)) {layerGroup=None}
  ]}
 EventGraph {nodes [
   Event BlueprintUpdateAnimation()
   > n1 = Sequence()
   n2 = GetOwnerSkeletalMeshComponent()
   n3 = BreakRotator(InRot=BreakTransform(InTransform=GetDeltaTransformFromRefPose(self=@n2, BoneName=thigh_r, BaseName=pelvis)).Rotation)
   > Set thigh_rot_up_r(thigh_rot_up_r:double=MapRangeClamped(Value=@n3.Yaw, InRangeA=-15, InRangeB=-100)) -> (Output_Get:double)
   > Set thigh_rot_out_r(thigh_rot_out_r:double=MapRangeClamped(Value=@n3.Pitch, InRangeA=15, InRangeB=40)) -> (Output_Get:double)
   > Set thigh_rot_innr_r(thigh_rot_innr_r:double=MapRangeClamped(Value=@n3.Pitch, InRangeA=0, InRangeB=-20)) -> (Output_Get:double)
   n4 = BreakTransform(InTransform=GetDeltaTransformFromRefPose(self=GetOwnerSkeletalMeshComponent(), BoneName=thigh_l, BaseName=pelvis))
   > Set thigh_rot_up_l(thigh_rot_up_l:double=MapRangeClamped(Value=BreakRotator(InRot=@n4.Rotation).Yaw, InRangeA=-10, InRangeB=-100)) -> (Output_Get:double)
   n5 = GetOwnerSkeletalMeshComponent()
   n6 = BreakRotator(InRot=BreakTransform(InTransform=GetDeltaTransformFromRefPose(self=@n5, BoneName=upperarm_r, BaseName=spine_05)).Rotation)
   > Set arm_rot_up_r(arm_rot_up_r:double=MapRangeClamped(Value="Absolute (Float)"(A=@n6.Yaw), InRangeA=0, InRangeB=135)) -> (Output_Get:double)
   n7 = GetOwnerSkeletalMeshComponent()
   n8 = BreakRotator(InRot=BreakTransform(InTransform=GetDeltaTransformFromRefPose(self=@n7, BoneName=hand_r, BaseName=lowerarm_twist_01_r)).Rotation)
   > Set hand_rot_up_r(hand_rot_up_r:double=MapRangeClamped(Value=@n8.Yaw, InRangeA=0, InRangeB=40)) -> (Output_Get:double)
   n9 = Branch(execute=@n1.then_1, Condition=Get bIsSkyDiving())
   > Set should_pad_def_alpha(should_pad_def_alpha:double=0) -> (Output_Get:double)
   > n10 = Branch(Condition=Get bIsSkydiveDiveMode())
   Set should_pad_def_alpha_dive(execute=[@n10.then, @n11.then], should_pad_def_alpha_dive:double=0) -> (Output_Get:double)
   Set should_pad_def_alpha_dive(execute=@n10.else, should_pad_def_alpha_dive=Get arm_rot_up_r() -> (arm_rot_up_r:double))
   n11 = Set should_pad_def_alpha(execute=@n9.else, should_pad_def_alpha=Get arm_rot_up_r())
   "Absolute (Float)"(A=@n8.Pitch)
  ]}
]
variables [
 arm_rot_up_r {type=number}
 hand_rot_up_r {type=number}
 should_pad_def_alpha {type=number}
 should_pad_def_alpha_dive {type=number}
 thigh_rot_innr_r {type=number}
 thigh_rot_out_r {type=number}
 thigh_rot_up_l {type=number}
 thigh_rot_up_r {type=number}
]
