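# Physics asset for the Primal Falcon female medium face accessory, providing bodies for dynamic hood and pendant bones on each side.
# Each side chains from neck_01 through a hood master, a tightly limited hood bone and two pendant bones with wide swing but little twist.
# Capsules on upper spine, neck, clavicles and upper arms form a body proxy, linked by constraints with free swing and twist.
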PhysicsAsset /BRCosmetics/Characters/Player/Female/Medium/Bodies/F_MED_Primal_Falcon/Meshes/Parts/F_MED_Primal_Falcon_FaceAcc_dyn
bodies {
 clavicle_l [{shapeType=Capsule, length=1.458, radius=6.4904, shapeName=spine_05_capsule, center=<6.4684 0.9489 0.6509>, rotation=<rot 90 0 0>}]
 clavicle_r [{
   shapeType=Capsule
   length=12.708
   radius=3.7404
   shapeName=spine_05_capsule
   center=<-9.6427 0.5347 0.5226>
   rotation=<rot 90 0.6714 -177.5614>
  }]
 dyn_hood_l [{shapeType=Capsule, length=1.9202, radius=2.1296, shapeName=dyn_pendant_r_01_capsule, center=<3 3 2>}]
 dyn_hood_l_MSTR [{shapeType=Capsule, length=0.505, radius=0.505, shapeName=dyn_hood_l_MSTR_capsule, rotation=<rot 0 -90 0>}]
 dyn_hood_r [{shapeType=Capsule, length=1.9202, radius=2.1296, shapeName=dyn_pendant_r_01_capsule, center=<3 -3 2>}]
 dyn_hood_r_MSTR [{shapeType=Capsule, length=0.505, radius=0.505, shapeName=dyn_hood_r_MSTR_capsule, rotation=<rot 0 -90 0>}]
 dyn_pendant_l_01 [{shapeType=Capsule, length=1.6702, radius=1.3796, shapeName=dyn_pendant_l_01_capsule, center=<1.5 0 0>, rotation=<rot 0 88.1142 -89.937>}]
 dyn_pendant_l_02 [{shapeType=Capsule, length=7.8253, radius=1.3905, shapeName=dyn_pendant_l_02_capsule, center=<6 0 0>, rotation=<rot 0 89.8101 -89.8924>}]
 dyn_pendant_r_01 [{shapeType=Capsule, length=1.6702, radius=1.3796, shapeName=dyn_pendant_r_01_capsule, center=<1.5 0 0>, rotation=<rot 0 91.8858 -89.937>}]
 dyn_pendant_r_02 [{shapeType=Capsule, length=7.8253, radius=1.3905, shapeName=dyn_pendant_r_02_capsule, center=<6 0 0>, rotation=<rot 0 90.1898 -89.8924>}]
 neck_01 [{shapeType=Capsule, length=3.458, radius=5.9904, shapeName=spine_05_capsule, center=<0 1.5165 0>, rotation=<rot 90 0 0>}]
 spine_04 [{shapeType=Capsule, length=9.958, radius=5.7404, shapeName=spine_05_capsule, center=<-0.0571 6.8288 0>}]
 spine_05 [{shapeType=Capsule, length=6.958, radius=11.7404, shapeName=spine_05_capsule, center=<-1.1603 -1.6874 0>}]
 upperarm_l [{
   shapeType=Capsule
   length=12.708
   radius=3.7404
   shapeName=spine_05_capsule
   center=<9.6427 -0.5347 -0.5226>
   rotation=<rot 90 -1.7899 -0.0228>
  }]
 upperarm_r [{
   shapeType=Capsule
   length=12.708
   radius=3.7404
   shapeName=spine_05_capsule
   center=<-9.6427 0.5347 0.5226>
   rotation=<rot 90 0.6714 -177.5614>
  }]
}
constraints [
 {bone1Name=dyn_pendant_r_02, bone2Name=dyn_pendant_r_01, swing1LimitDegrees=45, swing2LimitDegrees=45, twistLimitDegrees=2}
 {bone1Name=dyn_pendant_l_02, bone2Name=dyn_pendant_l_01, swing1LimitDegrees=45, swing2LimitDegrees=45, twistLimitDegrees=2}
 {
  bone1Name=neck_01
  bone2Name=spine_05
  swing1LimitDegrees=45
  swing1Motion=Free
  swing2LimitDegrees=45
  swing2Motion=Free
  twistLimitDegrees=45
  twistMotion=Free
 }
 {bone1Name=dyn_pendant_r_01, bone2Name=dyn_hood_r, swing1LimitDegrees=45, swing2LimitDegrees=45, twistLimitDegrees=2}
 {bone1Name=dyn_hood_r, bone2Name=dyn_hood_r_MSTR, swing1LimitDegrees=2, swing2LimitDegrees=2, twistLimitDegrees=10}
 {
  bone1Name=dyn_hood_r_MSTR
  bone2Name=neck_01
  swing1LimitDegrees=45
  swing1Motion=Free
  swing2LimitDegrees=45
  swing2Motion=Free
  twistLimitDegrees=45
  twistMotion=Free
 }
 {bone1Name=dyn_pendant_l_01, bone2Name=dyn_hood_l, swing1LimitDegrees=45, swing2LimitDegrees=45, twistLimitDegrees=2}
 {
  bone1Name=dyn_hood_l_MSTR
  bone2Name=neck_01
  swing1LimitDegrees=45
  swing1Motion=Free
  swing2LimitDegrees=45
  swing2Motion=Free
  twistLimitDegrees=45
  twistMotion=Free
 }
 {bone1Name=dyn_hood_l, bone2Name=dyn_hood_l_MSTR, swing1LimitDegrees=2, swing2LimitDegrees=2, twistLimitDegrees=10}
 {
  bone1Name=spine_05
  bone2Name=spine_04
  swing1LimitDegrees=45
  swing1Motion=Free
  swing2LimitDegrees=45
  swing2Motion=Free
  twistLimitDegrees=45
  twistMotion=Free
 }
 {
  bone1Name=upperarm_r
  bone2Name=clavicle_r
  swing1LimitDegrees=45
  swing1Motion=Free
  swing2LimitDegrees=45
  swing2Motion=Free
  twistLimitDegrees=45
  twistMotion=Free
 }
 {
  bone1Name=upperarm_l
  bone2Name=clavicle_l
  swing1LimitDegrees=45
  swing1Motion=Free
  swing2LimitDegrees=45
  swing2Motion=Free
  twistLimitDegrees=45
  twistMotion=Free
 }
 {
  bone1Name=clavicle_l
  bone2Name=spine_05
  swing1LimitDegrees=45
  swing1Motion=Free
  swing2LimitDegrees=45
  swing2Motion=Free
  twistLimitDegrees=45
  twistMotion=Free
 }
 {
  bone1Name=clavicle_r
  bone2Name=spine_05
  swing1LimitDegrees=45
  swing1Motion=Free
  swing2LimitDegrees=45
  swing2Motion=Free
  twistLimitDegrees=45
  twistMotion=Free
 }
]
