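# Post-copy AnimBP for the Primal Falcon female face accessory, animating its pendants, hood and feathers.
# Blends attached-parent pose with Body_Mesh upper arms, then applies pendant LookAts, additive right-hood roll, feather AnimDynamics and RigidBody simulation.
# EventGraph derives hood alpha from right upper-arm yaw delta relative to spine_05.
# LookAts and hood roll act only as fallback when bRigidBodyEnabled is false; both are zeroed otherwise.
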AnimBlueprint /BRCosmetics/Characters/Player/Female/Medium/Bodies/F_MED_Primal_Falcon/Meshes/Parts/F_MED_Primal_Falcon_FaceAcc_AnimBP
target_skeleton=&/F_MED_Primal_Falcon_FaceAcc_Skeleton
graphs [
 AnimGraph {nodes [
   n1 = Copy Pose From Mesh(SourceMeshComponent=Get Body_Mesh())
   n2 = Copy Pose From Mesh() {bCopyCurves=true, bUseAttachedParent=true}
   n3 = Get aim_Alpha() -> (aim_Alpha:bool)
   n4 = Look At - Bone: dyn_pendant_r_01!(
    ComponentPose=Local To Component(LocalPose=Layered blend per bone(BasePose=@n2, BlendPoses_0=@n1) {layerSetup [{branchFilters [{boneName=upperarm_l}, {boneName=upperarm_r}]}]})
    bAlphaBoolEnabled:bool=@n3
   ) -> (Pose) {
    alphaInputType=Bool
    alphaBoolBlend {blendOption=Linear}
    boneToModify {boneName=dyn_pendant_r_01}
    lookAtLocation=<0 11.5 6.5>
    lookAtTarget {boneReference {boneName=spine_05}}
    lookAt_Axis {axis=<1 0>}
    lookUp_Axis {axis {x=0}}
   }
   n5 = Look At - Bone: dyn_pendant_l_01!(ComponentPose=@n4, bAlphaBoolEnabled:bool=@n3) -> (Pose) {
    alphaInputType=Bool
    alphaBoolBlend {blendOption=Linear}
    boneToModify {boneName=dyn_pendant_l_01}
    lookAtLocation=<0 11.5 -6.5>
    lookAtTarget {boneReference {boneName=spine_05}}
    lookAt_Axis {axis=<1 0>}
    lookUp_Axis {axis {x=0}}
   }
   n6 = "Transform (Modify) Bone - Bone: dyn_hood_r_MSTR"(ComponentPose=@n5, Alpha=Get arm_rot_up_r() -> (arm_rot_up_r:double), Translation="0,0,0", Rotation="0.000,0.000,-15.001", Scale="1,1,1") {
    rotationMode=BMM_Additive
    rotationSpace=BCS_BoneSpace
    translationMode=BMM_Additive
    translationSpace=BCS_BoneSpace
    alphaBoolBlend {blendOption=Linear}
    boneToModify {boneName=dyn_hood_r_MSTR}
    rotation {roll=-15.0011}
   }
   n7 = Anim Dynamics - Bone: dyn_feathers_r(ComponentPose=@n6, Alpha=1, ExternalForce="0,0,0") {
    angularDampingOverride=0.9999
    bGravityOverrideInSimSpace=true
    bOverrideAngularDamping=true
    bUseGravityOverride=true
    linearDampingOverride=0.9
    simulationSpace=RootRelative
    alphaBoolBlend {blendOption=Linear}
    boundBone {boneName=dyn_feathers_r}
    componentAppliedLinearAccClamp {z=5000}
    componentLinearAccScale {z=0.4}
    componentLinearVelScale {z=0.5}
    gravityOverride {z=-980}
    physicsBodyDefinitions [{
      sphereCollisionRadius=0
      boundBone {boneName=dyn_feathers_r}
      boxExtents=<10 20 15>
      constraintSetup {angularLimitsMax=<0 30 10>, angularLimitsMin=<0 -30 -10>, angularTarget=<0 0 0>}
      localJointOffset=<4 0 0>
     }]
   }
   n8 = Anim Dynamics - Bone: dyn_feathers_l(ComponentPose=@n7, Alpha=1, ExternalForce="0,0,0") {
    angularDampingOverride=0.9999
    bGravityOverrideInSimSpace=true
    bOverrideAngularDamping=true
    bUseGravityOverride=true
    linearDampingOverride=0.9
    simulationSpace=RootRelative
    alphaBoolBlend {blendOption=Linear}
    boundBone {boneName=dyn_feathers_l}
    componentAppliedLinearAccClamp {z=3000}
    componentLinearAccScale {z=0.4}
    componentLinearVelScale {z=0.5}
    gravityOverride {z=-980}
    physicsBodyDefinitions [{
      sphereCollisionRadius=0
      boundBone {boneName=dyn_feathers_l}
      boxExtents=<10 20 15>
      constraintSetup {angularLimitsMax=<0 25 5>, angularLimitsMin=<0 -45 -5>, angularTarget=<0 0 0>}
      localJointOffset=<3 0 0>
     }]
   }
   n9 = RigidBody(ComponentPose=@n8) {
    bClampLinearTranslationLimitToRefPose=true
    bOverrideWorldGravity=true
    overridePhysicsAsset=&/F_MED_Primal_Falcon_FaceAcc_dyn
    simulationSpace=BaseBoneSpace
    alphaBoolBlend {blendOption=Linear}
    baseBoneRef {boneName=spine_05}
    componentAppliedLinearAccClamp=<3000 3000 3000>
    componentLinearAccScale=<0.3 0.3 0.3>
    componentLinearVelScale=<0.3 0.3 0.3>
    overrideWorldGravity {z=-980}
   }
   Output Pose(Result=Component To Local(ComponentPose=@n9)) {layerGroup=None}
  ]}
 EventGraph {nodes [
   Event BlueprintUpdateAnimation()
   n1 = GetOwnerSkeletalMeshComponent()
   n2 = BreakRotator(InRot=BreakTransform(InTransform=GetDeltaTransformFromRefPose(self=@n1, BoneName=upperarm_r, BaseName=spine_05)).Rotation)
   > Set arm_rot_up_r(arm_rot_up_r:double=MapRangeClamped(Value="Absolute (Float)"(A=@n2.Yaw), InRangeA=0, InRangeB=100)) -> (Output_Get:double)
   > n3 = Branch(Condition=Get bRigidBodyEnabled())
   > Set aim_Alpha(aim_Alpha:bool=false) -> (Output_Get:bool)
   > Set arm_rot_up_r(arm_rot_up_r=0)
   Set aim_Alpha(execute=@n3.else, aim_Alpha=true)
   > Set arm_rot_up_r(arm_rot_up_r=Get arm_rot_up_r())
   Event BlueprintInitializeAnimation()
   > Set Body_Mesh(Body_Mesh=GetOwnerSkeletalMeshComponent())
  ]}
]
variables [
 aim_Alpha {type=boolean}
 arm_rot_up_r {type=number}
 Body_Mesh {type=SkeletalMeshComponent}
]
